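# Multicolor emissive subsurface material for winter decoration lights, with R, G and B vertex colors masking three tints.
# Per-channel sines pulse the glow, offset by actor X position; static switches control animation, brightness clamping, global pulsing and hue cycling.
# Fresnel fades emissive toward edges, where tinted base color takes over.
# A shading path switch gives mobile a dimmer, Fresnel-free version before both paths enter MF_GameWideShared.
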Material /Game/Environments/Apollo/Sets/WinterDeco/Materials/M_MultiColorEmissive
graph {
 properties {
  bUseMaterialAttributes=true
  bUsedWithNanite=true
  pixelDepthOffsetMode=PDOM_Legacy
  shadingModel=MSM_Subsurface
  usedShadingModels=Subsurface
 }
 nodes [
  Constant3Vector()
  n1 = Time()
  n2 = ScalarParameter() {defaultValue=3.64, group=EmissiveLights, parameterName=ColorCycle_Speed}
  n3 = VectorParameter() {group=EmissiveLights, parameterName=R_Tint, sortPriority=1, defaultValue=<rgba 1 0 0 1>}
  n4 = VertexColor()
  n5 = LinearInterpolate(
   A=LinearInterpolate(
    A=LinearInterpolate(B=@n3.RGB, Alpha=@n4.R)
    B=VectorParameter() {group=EmissiveLights, parameterName=G_Tint, sortPriority=2, defaultValue=<rgba 0 1 0.0117 1>}.RGB
    Alpha=@n4.G
   )
   B=VectorParameter() {group=EmissiveLights, parameterName=B_Tint, sortPriority=3, defaultValue=<rgba 0 0.0578 1 1>}.RGB
   Alpha=@n4.B
  )
  n6 = StaticSwitchParameter(
   True=MaterialFunctionCall(Hue Shift Percentage=Divide(A=@n1, B=@n2) {constB=3.64}, Texture=@n5) {materialFunction=&HueShift}.Result
   False=@n5
  ) {group=EmissiveLights, parameterName=Enable_HueShift}
  n7 = Multiply(A=@n6, B=ScalarParameter() {defaultValue=0.25, group=EmissiveLights, parameterName=Diffuse_Intensity}) {constB=0.05}
  n8 = Constant3Vector()
  n9 = Fresnel(
   ExponentIn=ScalarParameter() {defaultValue=0.025, group=EmissiveLights, parameterName=Fresnel_Exponent}
   BaseReflectFractionIn=ScalarParameter() {defaultValue=0.4, group=EmissiveLights, parameterName=Fresnel_BaseReflect}
  )
  n10 = ScalarParameter() {defaultValue=0.25, group=EmissiveLights, parameterName=Metalness}
  n11 = ScalarParameter() {defaultValue=1, group=EmissiveLights, parameterName=Specular}
  n12 = ScalarParameter() {defaultValue=0.2, group=EmissiveLights, parameterName=Roughness}
  n13 = ScalarParameter() {defaultValue=75, group=EmissiveLights, parameterName=Emissive_Intensity}
  n14 = Multiply(A=@n6, B=@n13)
  n15 = Divide(A=Time(), B=@n2) {constB=3.64}
  n16 = VectorParameter() {group=EmissiveLights, parameterName=R_Tint, sortPriority=1, defaultValue=<rgba 1 0 0 1>}
  n17 = VertexColor()
  n18 = MaterialFunctionCall(Color=ActorPositionWS()) {materialFunction=&SplitComponents}
  n19 = Add(A=@n18.R, B=@n18.R)
  n20 = ScalarParameter() {defaultValue=20, group=EmissiveLights, parameterName=Offset_Interval, sortPriority=7}
  n21 = Add(A=Time(), B=Frac(None=Divide(A=@n19, B=@n20)))
  n22 = Sine(None=Divide(A=@n21, B=ScalarParameter() {defaultValue=5, group=EmissiveLights, parameterName=R_Frequency, sortPriority=4}))
  n23 = Clamp(None=@n22, Min=ScalarParameter() {group=EmissiveLights, parameterName=MinBrightness}) {clampMode=CMODE_ClampMin}
  n24 = StaticSwitchParameter(
   True=@n17.R
   False=Multiply(A=@n17.R, B=StaticSwitchParameter(True=@n23, False=@n22) {group=EmissiveLights, parameterName=ClampMinBrightness})
  ) {group=EmissiveLights, parameterName=DisableAnim_Red}
  n25 = VectorParameter() {group=EmissiveLights, parameterName=G_Tint, sortPriority=2, defaultValue=<rgba 0 1 0.0117 1>}
  n26 = Sine(None=Divide(A=@n21, B=ScalarParameter() {defaultValue=3.2, group=EmissiveLights, parameterName=G_Frequency, sortPriority=5}))
  n27 = Clamp(None=@n26, Min=ScalarParameter() {group=EmissiveLights, parameterName=MinBrightness}) {clampMode=CMODE_ClampMin}
  n28 = StaticSwitchParameter(
   True=@n17.G
   False=Multiply(A=@n17.G, B=StaticSwitchParameter(True=@n27, False=@n26) {group=EmissiveLights, parameterName=ClampMinBrightness})
  ) {group=EmissiveLights, parameterName=DisableAnim_Green}
  n29 = Sine(None=Divide(A=@n21, B=ScalarParameter() {defaultValue=2.1, group=EmissiveLights, parameterName=B_Frequency, sortPriority=6}))
  n30 = ScalarParameter() {group=EmissiveLights, parameterName=MinBrightness}
  n31 = Multiply(
   A=@n17.B
   B=StaticSwitchParameter(True=Clamp(None=@n29, Min=@n30) {clampMode=CMODE_ClampMin}, False=@n29) {group=EmissiveLights, parameterName=ClampMinBrightness}
  )
  n32 = Add(
   A=Add(A=LinearInterpolate(B=@n16.RGB, Alpha=@n24), B=LinearInterpolate(B=@n25.RGB, Alpha=@n28))
   B=LinearInterpolate(
    B=VectorParameter() {group=EmissiveLights, parameterName=B_Tint, sortPriority=3, defaultValue=<rgba 0 0.0578 1 1>}.RGB
    Alpha=StaticSwitchParameter(True=@n17.B, False=@n31) {group=EmissiveLights, parameterName=DisableAnim_Blue}
   )
  )
  n33 = Multiply(
   A=StaticSwitchParameter(
    True=MaterialFunctionCall(Hue Shift Percentage=@n15, Texture=@n32) {materialFunction=&HueShift}.Result
    False=@n32
   ) {group=EmissiveLights, parameterName=Enable_HueShift}
   B=@n13
  )
  n34 = Time()
  n35 = ScalarParameter() {defaultValue=5, group=EmissiveLights, parameterName=Emissive_RGB_TimeInterval}
  n36 = Sine(None=Frac(None=Divide(A=@n34, B=@n35)))
  n37 = Clamp(None=@n36, Min=ScalarParameter() {group=EmissiveLights, parameterName=MinBrightness}) {clampMode=CMODE_ClampMin}
  n38 = StaticSwitchParameter(True=Multiply(A=@n33, B=StaticSwitchParameter(True=@n37, False=@n36) {group=EmissiveLights, parameterName=ClampMinBrightness}), False=@n33) {group=EmissiveLights, parameterName=Emissive_Intensity_Offset, sortPriority=33}
  n39 = SetMaterialAttributes!(
   MaterialAttributes
   Base Color=LinearInterpolate(A=@n7, B=@n8, Alpha=OneMinus(None=@n9))
   Metallic=@n10
   Specular=@n11
   Roughness=@n12
   Emissive Color=LinearInterpolate(A=StaticSwitchParameter(True=@n14, False=@n38) {group=EmissiveLights, parameterName=Static_Emissive_AllChannels}, B=@n8, Alpha=@n9)
   Subsurface Color=Multiply(A=@n6, B=ScalarParameter() {defaultValue=0.03, group=EmissiveLights, parameterName=SS_Color_Intensity}) {constB=0.05}
  ) {attributeSetTypes [
    #1
    #2
    #3
    #4
    #5
    #6
   ]}
  n40 = SetMaterialAttributes!(
   MaterialAttributes
   Base Color=@n7
   Metallic=@n10
   Specular=@n11
   Roughness=@n12
   Emissive Color=Multiply(
    A=StaticSwitchParameter(True=@n6, False=@n38) {group=EmissiveLights, parameterName=Static_Emissive_AllChannels}
    B=ScalarParameter() {defaultValue=0.01, parameterName=Mobile_Emissive_Intensity}
   )
  ) {attributeSetTypes [
    #1
    #2
    #3
    #4
    #5
   ]}
  MaterialFunctionCall!(In=ShadingPathSwitch(Default=@n39, Mobile=@n40)) {materialFunction=/Game/Athena/WinterArt/Materials/MF_GameWideShared}
 ]
}
